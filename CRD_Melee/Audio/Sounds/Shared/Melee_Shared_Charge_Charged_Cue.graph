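# Shared melee weapon SoundCue for reaching full charge, layering start one-shots over a sustained charged loop.
# It mixes a random fully-charged start and a briefly delayed loop, shaped by pitch and volume envelopes, with a modulated random charge-start layer.
# Modulators vary each playback, and a concurrency override allows only one instance at a time under the melee sound class and a small-toy attenuation.
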SoundCue /CRD_Melee/Audio/Sounds/Shared/Melee_Shared_Charge_Charged_Cue
root=Mixer_3
nodes [
 {id=Delay_1, children [Enveloper_1], properties {delayMax=0.1, delayMin=0.1}}
 {
  id=Enveloper_1
  children [WavePlayer_5]
  properties {pitchCurve {editorCurveData {keys [<at 0 0.95>, {time=2}]}}, volumeCurve {editorCurveData {keys [<at 2 1>,<at 2.5 0.7>]}}}
 }
 {id=Mixer_2, children [Random_1,Delay_1], properties {inputVolume [0.5,0.8]}}
 {id=Mixer_3, children [Modulator_2,Modulator_0], properties {inputVolume [1,1]}}
 {id=Modulator_0, children [Random_2]}
 {id=Modulator_2, children [Mixer_2], properties {volumeMax=0.95, volumeMin=0.85}}
 {id=Random_1, children [WavePlayer_1,WavePlayer_4], properties {weights [1,1]}}
 {id=Random_2, children [WavePlayer_6,WavePlayer_7,WavePlayer_8], properties {weights [1,1,1]}}
 {id=WavePlayer_1, properties {soundWaveAssetPtr=&/Melee_Shared_FullyCharged_Start_01}}
 {id=WavePlayer_4, properties {soundWaveAssetPtr=&/Melee_Shared_FullyCharged_Start_02}}
 {id=WavePlayer_5, properties {bLooping=true, soundWaveAssetPtr=&/Melee_Shared_FullyCharged_Loop_01}}
 {id=WavePlayer_6, properties {soundWaveAssetPtr=&/Melee_Shared_Charge_Start_V2_01}}
 {id=WavePlayer_7, properties {soundWaveAssetPtr=&/Melee_Shared_Charge_Start_V2_02}}
 {id=WavePlayer_8, properties {soundWaveAssetPtr=&/Melee_Shared_Charge_Start_V2_03}}
]
properties {
 attenuationSettings=/Game/Sounds/Creative/Attenuations/Fort_Creative_SmallToy_Att
 bOverrideConcurrency=true
 maxDistance=3000
 soundClassObject=&Fort_WeaponMelee
 volumeMultiplier=0.8
 concurrencyOverrides {maxCount=1}
}
