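# Shared melee charge sound cue that layers a charge-start one-shot over a looping charge sound.
# A Random node picks one of three start waves and mixes it with the looping wave, all under a root Modulator.
# An Enveloper shapes only the loop, moving pitch over the first two seconds and then easing volume down.
# Concurrency is capped at one instance, using small-toy attenuation and the melee weapon sound class.
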SoundCue /CRD_Melee/Audio/Sounds/Shared/Melee_Shared_Charge_Loop_Cue
root=Modulator_0
nodes [
 {
  id=Enveloper_0
  children [WavePlayer_0]
  properties {pitchCurve {editorCurveData {keys [<at 0 0.95>, {time=2}]}}, volumeCurve {editorCurveData {keys [<at 2 1>,<at 2.5 0.7>]}}}
 }
 {id=Mixer_0, children [Random_1,Enveloper_0], properties {inputVolume [1,1]}}
 {id=Modulator_0, children [Mixer_0]}
 {id=Random_1, children [WavePlayer_1,WavePlayer_4,WavePlayer_5], properties {weights [1,1,1]}}
 {id=WavePlayer_0, properties {bLooping=true, soundWaveAssetPtr=&/Melee_Shared_Charge_Loop_V2_01}}
 {id=WavePlayer_1, properties {soundWaveAssetPtr=&/Melee_Shared_Charge_Start_01}}
 {id=WavePlayer_4, properties {soundWaveAssetPtr=&/Melee_Shared_Charge_Start_02}}
 {id=WavePlayer_5, properties {soundWaveAssetPtr=&/Melee_Shared_Charge_Start_03}}
]
properties {
 attenuationSettings=/Game/Sounds/Creative/Attenuations/Fort_Creative_SmallToy_Att
 bOverrideConcurrency=true
 maxDistance=3000
 pitchMultiplier=0.9
 soundClassObject=&Fort_WeaponMelee
 volumeMultiplier=0.6
 concurrencyOverrides {maxCount=1}
}
