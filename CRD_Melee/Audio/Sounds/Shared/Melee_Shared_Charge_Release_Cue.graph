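# Shared melee sound cue for releasing a charge, classed as weapon melee audio with small-toy attenuation.
# A Branch on the IsCharged bool picks one of two fully-charged stop waves when true, or one of three charge-release waves when false or unset.
# Each set is chosen by an equal-weight Random node and passes through a default Modulator for variation.
# A concurrency override limits the cue to one playing instance.
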SoundCue /CRD_Melee/Audio/Sounds/Shared/Melee_Shared_Charge_Release_Cue
root=Branch_0
nodes [
 {id=Branch_0, children [Modulator_2,Modulator_1,Modulator_1], properties {boolParameterName=IsCharged}}
 {id=Modulator_1, children [Random_0]}
 {id=Modulator_2, children [Random_1]}
 {id=Random_0, children [WavePlayer_1,WavePlayer_2,WavePlayer_3], properties {weights [1,1,1]}}
 {id=Random_1, children [WavePlayer_0,WavePlayer_4], properties {weights [1,1]}}
 {id=WavePlayer_0, properties {soundWaveAssetPtr=&/Melee_Shared_FullyCharged_Stop_01}}
 {id=WavePlayer_1, properties {soundWaveAssetPtr=&/Melee_Shared_Charge_Release_V2_01}}
 {id=WavePlayer_2, properties {soundWaveAssetPtr=&/Melee_Shared_Charge_Release_V2_02}}
 {id=WavePlayer_3, properties {soundWaveAssetPtr=&/Melee_Shared_Charge_Release_V2_03}}
 {id=WavePlayer_4, properties {soundWaveAssetPtr=&/Melee_Shared_FullyCharged_Stop_02}}
]
properties {
 attenuationSettings=/Game/Sounds/Creative/Attenuations/Fort_Creative_SmallToy_Att
 bOverrideConcurrency=true
 maxDistance=3000
 soundClassObject=&Fort_WeaponMelee
 volumeMultiplier=0.6
 concurrencyOverrides {maxCount=1}
}
