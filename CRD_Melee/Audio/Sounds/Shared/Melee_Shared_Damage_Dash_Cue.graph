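# Sound cue for shared melee dash damage, played with a close-range melee weapon attenuation and the Fort weapon melee sound class.
# A Random node picks one of three dash damage waves with equal weight, then feeds a Modulator at the root.
# Playback is limited by both a melee damage concurrency and a global melee damage concurrency, with a slightly reduced volume multiplier.
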SoundCue /CRD_Melee/Audio/Sounds/Shared/Melee_Shared_Damage_Dash_Cue
root=Modulator_1
nodes [
 {id=Modulator_1, children [Random_0]}
 {id=Random_0, children [WavePlayer_1,WavePlayer_2,WavePlayer_3], properties {weights [1,1,1]}}
 {id=WavePlayer_1, properties {soundWaveAssetPtr=&/Melee_Shared_Damage_Dash_01}}
 {id=WavePlayer_2, properties {soundWaveAssetPtr=&/Melee_Shared_Damage_Dash_02}}
 {id=WavePlayer_3, properties {soundWaveAssetPtr=&/Melee_Shared_Damage_Dash_03}}
]
properties {
 attenuationSettings=/Game/Sounds/Attenuation/Fort_Weapon_Melee_Attack_Close_Range_Stereo_Att
 maxDistance=10058.75
 soundClassObject=&Fort_WeaponMelee
 volumeMultiplier=0.9
 concurrencySet [/CRD_Melee/Audio/Settings/Melee_Damage_Conc, /CRD_Melee/Audio/Settings/Melee_Damage_Global_Conc]
}
