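# Sound cue for the shared heavy melee damage sound in the CRD_Melee content.
# A Random node picks one of three equally weighted heavy damage waves and feeds a Modulator at the root for playback variation.
# It plays in the melee weapon sound class with close-range stereo melee attack attenuation, slightly reduced volume, and two damage concurrency settings, one of them named as global.
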SoundCue /CRD_Melee/Audio/Sounds/Shared/Melee_Shared_Damage_Heavy_Cue
root=Modulator_1
nodes [
 {id=Modulator_1, children [Random_1]}
 {id=Random_1, children [WavePlayer_0,WavePlayer_4,WavePlayer_5], properties {weights [1,1,1]}}
 {id=WavePlayer_0, properties {soundWaveAssetPtr=&/Melee_Shared_Damage_Heavy_01}}
 {id=WavePlayer_4, properties {soundWaveAssetPtr=&/Melee_Shared_Damage_Heavy_02}}
 {id=WavePlayer_5, properties {soundWaveAssetPtr=&/Melee_Shared_Damage_Heavy_03}}
]
properties {
 attenuationSettings=/Game/Sounds/Attenuation/Fort_Weapon_Melee_Attack_Close_Range_Stereo_Att
 maxDistance=10058.75
 soundClassObject=&Fort_WeaponMelee
 volumeMultiplier=0.9
 concurrencySet [/CRD_Melee/Audio/Settings/Melee_Damage_Conc, /CRD_Melee/Audio/Settings/Melee_Damage_Global_Conc]
}
