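# Sound cue for the shared melee defense-ready sound in the CRD_Melee content.
# A random node picks one of three equally weighted wave variations and feeds it through a modulator for pitch and volume variation.
# It plays quietly in the melee weapon sound class, with basic foley attenuation and a global defensive melee concurrency setting that governs overlapping instances.
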SoundCue /CRD_Melee/Audio/Sounds/Shared/Melee_Shared_Defense_Ready_Cue
root=Modulator_1
nodes [
 {id=Modulator_1, children [Random_0]}
 {id=Random_0, children [WavePlayer_1,WavePlayer_2,WavePlayer_3], properties {weights [1,1,1]}}
 {id=WavePlayer_1, properties {soundWaveAssetPtr=&/Melee_Shared_Defense_Ready_01}}
 {id=WavePlayer_2, properties {soundWaveAssetPtr=&/Melee_Shared_Defense_Ready_02}}
 {id=WavePlayer_3, properties {soundWaveAssetPtr=&/Melee_Shared_Defense_Ready_03}}
]
properties {
 attenuationSettings=&Foley_Attenuation_Basic_Att
 maxDistance=1536
 soundClassObject=&Fort_WeaponMelee
 volumeMultiplier=0.1
 concurrencySet [/CRD_Melee/Audio/Settings/Melee_Defensive_Global_Conc]
}
