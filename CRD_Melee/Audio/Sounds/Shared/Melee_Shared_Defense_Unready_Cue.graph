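# Sound cue for the shared melee defense "unready" sound, likely played when a character drops a defensive stance.
# A random node picks one of three wave variations with equal weight, then a modulator at the root varies the result.
# It plays quietly through foley attenuation and the melee weapon sound class, limited by a global defensive melee concurrency setting.
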SoundCue /CRD_Melee/Audio/Sounds/Shared/Melee_Shared_Defense_Unready_Cue
root=Modulator_1
nodes [
 {id=Modulator_1, children [Random_1]}
 {id=Random_1, children [WavePlayer_0,WavePlayer_4,WavePlayer_5], properties {weights [1,1,1]}}
 {id=WavePlayer_0, properties {soundWaveAssetPtr=&/Melee_Shared_Defense_Unready_01}}
 {id=WavePlayer_4, properties {soundWaveAssetPtr=&/Melee_Shared_Defense_Unready_02}}
 {id=WavePlayer_5, properties {soundWaveAssetPtr=&/Melee_Shared_Defense_Unready_03}}
]
properties {
 attenuationSettings=&Foley_Attenuation_Basic_Att
 maxDistance=1536
 soundClassObject=&Fort_WeaponMelee
 volumeMultiplier=0.1
 concurrencySet [/CRD_Melee/Audio/Settings/Melee_Defensive_Global_Conc]
}
