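# Sound cue for the shared melee "device added" sound in the CRD_Melee content.
# A Random node picks one of three DeviceAdded wave variations with equal weight.
# It plays quietly in the weapon melee sound class, using close-range pickaxe attenuation to limit how far it carries.
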SoundCue /CRD_Melee/Audio/Sounds/Shared/Melee_Shared_DeviceAdded_Cue
root=Random_1
nodes [
 {id=Random_1, children [WavePlayer_2,WavePlayer_3,WavePlayer_1], properties {weights [1,1,1]}}
 {id=WavePlayer_1, properties {soundWaveAssetPtr=&/Melee_Shared_DeviceAdded_03}}
 {id=WavePlayer_2, properties {soundWaveAssetPtr=&/Melee_Shared_DeviceAdded_01}}
 {id=WavePlayer_3, properties {soundWaveAssetPtr=&/Melee_Shared_DeviceAdded_02}}
]
properties {
 attenuationSettings=/Game/Sounds/Attenuation/Fort_Weapon_Pickaxe_Close_Att
 maxDistance=1536
 soundClassObject=&Fort_WeaponMelee
 volumeMultiplier=0.25
}
