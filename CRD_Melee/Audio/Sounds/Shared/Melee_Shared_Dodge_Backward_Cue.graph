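# Sound cue for the shared melee backward dodge in the CRD_Melee plugin.
# A random node picks among four equally weighted waves, mixing two backward-specific dodge sounds with two generic dodge sounds.
# A modulator at the root varies each play, with a lowered minimum pitch.
# It plays at reduced volume in the melee weapon sound class, using close-range melee attenuation and a defensive-melee global concurrency set.
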SoundCue /CRD_Melee/Audio/Sounds/Shared/Melee_Shared_Dodge_Backward_Cue
root=Modulator_1
nodes [
 {id=Modulator_1, children [Random_0], properties {pitchMin=0.85}}
 {id=Random_0, children [WavePlayer_1,WavePlayer_2,WavePlayer_6,WavePlayer_7], properties {weights [1,1,1,1]}}
 {id=WavePlayer_1, properties {soundWaveAssetPtr=&/Melee_Shared_Dodge_Backward_01}}
 {id=WavePlayer_2, properties {soundWaveAssetPtr=&/Melee_Shared_Dodge_Backward_02}}
 {id=WavePlayer_6, properties {soundWaveAssetPtr=&/Melee_Shared_Dodge_02}}
 {id=WavePlayer_7, properties {soundWaveAssetPtr=&/Melee_Shared_Dodge_03}}
]
properties {
 attenuationSettings=/Game/Sounds/Attenuation/Fort_Weapon_Melee_Attack_Close_Range_Stereo_Att
 maxDistance=10058.75
 soundClassObject=&Fort_WeaponMelee
 volumeMultiplier=0.7
 concurrencySet [/CRD_Melee/Audio/Settings/Melee_Defensive_Global_Conc]
}
