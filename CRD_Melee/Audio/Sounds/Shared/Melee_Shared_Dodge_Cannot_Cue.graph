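# Sound cue for the shared melee "cannot dodge" sound, likely feedback for an unavailable dodge.
# A mixer layers two wave players, with the second variant slightly quieter, at reduced overall volume in the melee weapon sound class.
# Close-range stereo melee attenuation spatializes it.
# A concurrency override allows one instance and blocks new ones, taking precedence over the assigned global defensive concurrency set.
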SoundCue /CRD_Melee/Audio/Sounds/Shared/Melee_Shared_Dodge_Cannot_Cue
root=Mixer_1
nodes [
 {id=Mixer_1, children [WavePlayer_0,WavePlayer_2], properties {inputVolume [1,0.7]}}
 {id=WavePlayer_0, properties {soundWaveAssetPtr=&/Melee_Shared_Dodge_Cannot_01}}
 {id=WavePlayer_2, properties {soundWaveAssetPtr=&/Melee_Shared_Dodge_Cannot_03}}
]
properties {
 attenuationSettings=/Game/Sounds/Attenuation/Fort_Weapon_Melee_Attack_Close_Range_Stereo_Att
 bOverrideConcurrency=true
 maxDistance=10058.75
 soundClassObject=&Fort_WeaponMelee
 volumeMultiplier=0.55
 concurrencyOverrides {maxCount=1, resolutionRule=PreventNew}
 concurrencySet [/CRD_Melee/Audio/Settings/Melee_Defensive_Global_Conc]
}
