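# Sound cue for the shared melee heavy attack jump sound in the CRD_Melee content.
# A Random node picks one of three equally weighted wave variations and feeds a root Modulator that varies pitch on playback.
# The cue uses the melee weapon sound class, close-range stereo melee attack attenuation and a global melee attack concurrency setting to manage overlapping plays.
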SoundCue /CRD_Melee/Audio/Sounds/Shared/Melee_Shared_HeavyAttackJump_Cue
root=Modulator_1
nodes [
 {id=Modulator_1, children [Random_2], properties {pitchMax=1.1}}
 {id=Random_2, children [WavePlayer_6,WavePlayer_7,WavePlayer_8], properties {weights [1,1,1]}}
 {id=WavePlayer_6, properties {soundWaveAssetPtr=&/Melee_Shared_HeavyAttackJump_01}}
 {id=WavePlayer_7, properties {soundWaveAssetPtr=&/Melee_Shared_HeavyAttackJump_02}}
 {id=WavePlayer_8, properties {soundWaveAssetPtr=&/Melee_Shared_HeavyAttackJump_03}}
]
properties {
 attenuationSettings=/Game/Sounds/Attenuation/Fort_Weapon_Melee_Attack_Close_Range_Stereo_Att
 maxDistance=10058.75
 soundClassObject=&Fort_WeaponMelee
 volumeMultiplier=1
 concurrencySet [/CRD_Melee/Audio/Settings/Melee_Attack_Global_Conc]
}
