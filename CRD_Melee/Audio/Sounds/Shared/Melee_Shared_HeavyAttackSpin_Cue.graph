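# Sound cue for the shared melee heavy attack spin, built from two randomized layers played together.
# A Mixer blends a random pick from waves 05–08 with a quieter pick from waves 01–04, which has its own Modulator for wider volume variation.
# A root Modulator varies the combined result.
# It plays in the melee weapon sound class with close-range stereo attenuation and a global melee attack concurrency limit.
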SoundCue /CRD_Melee/Audio/Sounds/Shared/Melee_Shared_HeavyAttackSpin_Cue
root=Modulator_1
nodes [
 {id=Mixer_2, children [Random_4,Modulator_2], properties {inputVolume [1,0.7]}}
 {id=Modulator_1, children [Mixer_2]}
 {id=Modulator_2, children [Random_5], properties {volumeMax=1.1, volumeMin=0.7}}
 {id=Random_4, children [WavePlayer_33, WavePlayer_34, WavePlayer_35, WavePlayer_36], properties {weights [1,1,1,1]}}
 {id=Random_5, children [WavePlayer_37, WavePlayer_38, WavePlayer_39, WavePlayer_40], properties {weights [1,1,1,1]}}
 {id=WavePlayer_33, properties {soundWaveAssetPtr=&/Melee_Shared_HeavyAttackSpin_05}}
 {id=WavePlayer_34, properties {soundWaveAssetPtr=&/Melee_Shared_HeavyAttackSpin_07}}
 {id=WavePlayer_35, properties {soundWaveAssetPtr=&/Melee_Shared_HeavyAttackSpin_06}}
 {id=WavePlayer_36, properties {soundWaveAssetPtr=&/Melee_Shared_HeavyAttackSpin_08}}
 {id=WavePlayer_37, properties {soundWaveAssetPtr=&/Melee_Shared_HeavyAttackSpin_01}}
 {id=WavePlayer_38, properties {soundWaveAssetPtr=&/Melee_Shared_HeavyAttackSpin_03}}
 {id=WavePlayer_39, properties {soundWaveAssetPtr=&/Melee_Shared_HeavyAttackSpin_02}}
 {id=WavePlayer_40, properties {soundWaveAssetPtr=&/Melee_Shared_HeavyAttackSpin_04}}
]
properties {
 attenuationSettings=/Game/Sounds/Attenuation/Fort_Weapon_Melee_Attack_Close_Range_Stereo_Att
 maxDistance=10058.75
 soundClassObject=&Fort_WeaponMelee
 volumeMultiplier=0.6
 concurrencySet [/CRD_Melee/Audio/Settings/Melee_Attack_Global_Conc]
}
