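# Sound cue for the landing of the shared melee primary jump, pairing a land sound with an explosion layer.
# A Mixer blends two branches at different volumes, each through its own Modulator, with random picks from three waves apiece.
# The jump-land branch is briefly delayed behind the explosion branch.
# It uses the close-range melee attack attenuation, the melee weapon sound class and a global melee attack concurrency.
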SoundCue /CRD_Melee/Audio/Sounds/Shared/Melee_Shared_Primary_Jump_Land_Cue
root=Mixer_0
nodes [
 {id=Delay_0, children [Random_0], properties {delayMax=0.1, delayMin=0.1}}
 {id=Mixer_0, children [Modulator_1,Modulator_0], properties {inputVolume [0.3,0.5]}}
 {id=Modulator_0, children [Random_1]}
 {id=Modulator_1, children [Delay_0]}
 {id=Random_0, children [WavePlayer_1,WavePlayer_2,WavePlayer_3], properties {weights [1,1,1]}}
 {id=Random_1, children [WavePlayer_0,WavePlayer_4,WavePlayer_5], properties {weights [1,1,1]}}
 {id=WavePlayer_0, properties {soundWaveAssetPtr=&/Melee_Shared_Primary_Land_Explosion_01}}
 {id=WavePlayer_1, properties {soundWaveAssetPtr=&/Melee_Shared_Primary_Jump_Land_01}}
 {id=WavePlayer_2, properties {soundWaveAssetPtr=&/Melee_Shared_Primary_Jump_Land_02}}
 {id=WavePlayer_3, properties {soundWaveAssetPtr=&/Melee_Shared_Primary_Jump_Land_03}}
 {id=WavePlayer_4, properties {soundWaveAssetPtr=&/Melee_Shared_Primary_Land_Explosion_02}}
 {id=WavePlayer_5, properties {soundWaveAssetPtr=&/Melee_Shared_Primary_Land_Explosion_03}}
]
properties {
 attenuationSettings=/Game/Sounds/Attenuation/Fort_Weapon_Melee_Attack_Close_Range_Stereo_Att
 maxDistance=10058.75
 soundClassObject=&Fort_WeaponMelee
 volumeMultiplier=1
 concurrencySet [/CRD_Melee/Audio/Settings/Melee_Attack_Global_Conc]
}
