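# Sound cue for the shared melee secondary dash, pairing a dash sound with a delayed landing sound.
# A mixer layers a random pick of three dash waves with a random pick of three primary jump-land waves, the latter at half volume after a fixed short delay.
# The mix passes through a root modulator and uses the melee weapon sound class, close-range melee attenuation and a global melee attack concurrency.
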SoundCue /CRD_Melee/Audio/Sounds/Shared/Melee_Shared_Secondary_Dash_Cue
root=Modulator_1
nodes [
 {id=Delay_0, children [Modulator_3], properties {delayMax=0.6, delayMin=0.6}}
 {id=Mixer_1, children [Delay_0,Random_1], properties {inputVolume [1,1]}}
 {id=Modulator_1, children [Mixer_1]}
 {id=Modulator_3, children [Random_0], properties {volumeMax=0.5, volumeMin=0.5}}
 {id=Random_0, children [WavePlayer_1,WavePlayer_2,WavePlayer_3], properties {weights [1,1,1]}}
 {id=Random_1, children [WavePlayer_0,WavePlayer_4,WavePlayer_5], properties {weights [1,1,1]}}
 {id=WavePlayer_0, properties {soundWaveAssetPtr=&/Melee_Shared_Secondary_Dash_03}}
 {id=WavePlayer_1, properties {soundWaveAssetPtr=&/Melee_Shared_Primary_Jump_Land_01}}
 {id=WavePlayer_2, properties {soundWaveAssetPtr=&/Melee_Shared_Primary_Jump_Land_02}}
 {id=WavePlayer_3, properties {soundWaveAssetPtr=&/Melee_Shared_Primary_Jump_Land_03}}
 {id=WavePlayer_4, properties {soundWaveAssetPtr=&/Melee_Shared_Secondary_Dash_01}}
 {id=WavePlayer_5, properties {soundWaveAssetPtr=&/Melee_Shared_Secondary_Dash_02}}
]
properties {
 attenuationSettings=/Game/Sounds/Attenuation/Fort_Weapon_Melee_Attack_Close_Range_Stereo_Att
 maxDistance=10058.75
 soundClassObject=&Fort_WeaponMelee
 volumeMultiplier=1
 concurrencySet [/CRD_Melee/Audio/Settings/Melee_Attack_Global_Conc]
}
