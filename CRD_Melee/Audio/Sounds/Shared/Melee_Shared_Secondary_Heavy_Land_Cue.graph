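# Sound cue for the shared melee secondary heavy attack landing, classed as melee weapon audio.
# It mixes a random pick from three main land sounds with a random pick from three layer sounds, the layer delayed a fixed tenth of a second.
# A top-level modulator varies the result, with close-range stereo melee attenuation and a global melee attack concurrency limit applied.
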SoundCue /CRD_Melee/Audio/Sounds/Shared/Melee_Shared_Secondary_Heavy_Land_Cue
root=Modulator_1
nodes [
 {id=Delay_0, children [Random_4], properties {delayMax=0.1, delayMin=0.1}}
 {id=Mixer_0, children [Random_0,Delay_0], properties {inputVolume [1,1]}}
 {id=Modulator_1, children [Mixer_0]}
 {id=Random_0, children [WavePlayer_1,WavePlayer_2,WavePlayer_3], properties {weights [1,1,1]}}
 {id=Random_4, children [WavePlayer_14, WavePlayer_15, WavePlayer_16], properties {weights [1,1,1]}}
 {id=WavePlayer_1, properties {soundWaveAssetPtr=&/Melee_Shared_Secondary_Heavy_Land_01}}
 {id=WavePlayer_14, properties {soundWaveAssetPtr=&/Melee_Shared_Secondary_Heavy_Land_Layer_01}}
 {id=WavePlayer_15, properties {soundWaveAssetPtr=&/Melee_Shared_Secondary_Heavy_Land_Layer_02}}
 {id=WavePlayer_16, properties {soundWaveAssetPtr=&/Melee_Shared_Secondary_Heavy_Land_Layer_03}}
 {id=WavePlayer_2, properties {soundWaveAssetPtr=&/Melee_Shared_Secondary_Heavy_Land_02}}
 {id=WavePlayer_3, properties {soundWaveAssetPtr=&/Melee_Shared_Secondary_Heavy_Land_03}}
]
properties {
 attenuationSettings=/Game/Sounds/Attenuation/Fort_Weapon_Melee_Attack_Close_Range_Stereo_Att
 maxDistance=10058.75
 soundClassObject=&Fort_WeaponMelee
 volumeMultiplier=1
 concurrencySet [/CRD_Melee/Audio/Settings/Melee_Attack_Global_Conc]
}
